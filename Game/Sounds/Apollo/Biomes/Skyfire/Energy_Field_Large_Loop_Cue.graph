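# Looping ambience cue for a large energy field in the Apollo Skyfire biome, routed to the in-game ambience sound class.
# A looping close-range wave feeds a Modulator with fixed volume, then an Enveloper that ramps volume in from silence.
# A Quality Level root sends that chain to its first two levels and leaves the other two empty.
# An attenuation asset handles spatial falloff, while overridden concurrency caps instances at three, stopping the quietest.
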SoundCue /Game/Sounds/Apollo/Biomes/Skyfire/Energy_Field_Large_Loop_Cue
root=QualityLevel_0
nodes [
 {
  id=Enveloper_0
  children [Modulator_0]
  properties {pitchCurve {editorCurveData {keys [{time=0}]}}, volumeCurve {editorCurveData {keys [<at 0 0 RCIM_Cubic>,<at 1 1>]}}}
 }
 {id=Modulator_0, children [WavePlayer_0], properties {volumeMax=1, volumeMin=1}}
 {id=QualityLevel_0, children [Enveloper_0,Enveloper_0,null,null]}
 {id=WavePlayer_0, properties {bLooping=true, soundWaveAssetPtr=&/Energy_Field_Large_Close_Loop}}
]
properties {
 attenuationSettings=&/Energy_Field_Loop_Large_Close_Att
 bOverrideConcurrency=true
 maxDistance=3200
 priority=0.5
 soundClassObject=/Game/Sounds/FortSoundClassesAndModes/Fort_LTM_BR_InGame_Ambience
 volumeMultiplier=2
 attenuationOverrides {
  bEnableFocusInterpolation=true
  bEnableSubmixSends=true
  distanceAlgorithm=Custom
  falloffDistance=2000
  focusAttackInterpSpeed=10
  focusAzimuth=90
  focusReleaseInterpSpeed=20
  focusVolumeAttenuation=1.75
  nonFocusAzimuth=170
  nonFocusVolumeAttenuation=0.2
  reverbDistanceMax=500
  reverbDistanceMin=50
  stereoSpread=1000
  attenuationShapeExtents {x=100}
  customAttenuationCurve {editorCurveData {keys [{arriveTangent=-2.8438, interpMode=RCIM_Cubic, leaveTangent=-2.8438, tangentMode=RCTM_User, time=0, value=1}, <at 1 0>]}}
 }
 concurrencyOverrides {maxCount=3, resolutionRule=StopQuietest}
}
