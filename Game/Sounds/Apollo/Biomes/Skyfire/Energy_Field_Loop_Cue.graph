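# Looping close-range energy field ambience cue for the Skyfire biome, played through the in-game ambience sound class.
# A looping wave passes through a Modulator for slight pitch variation, then an Enveloper that fades volume in from silence.
# The Quality Level root shares that branch across its first two slots and leaves the other two empty.
# A dedicated attenuation asset shapes falloff, and a concurrency override caps instances at three, stopping the quietest.
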SoundCue /Game/Sounds/Apollo/Biomes/Skyfire/Energy_Field_Loop_Cue
root=QualityLevel_0
nodes [
 {
  id=Enveloper_0
  children [Modulator_0]
  properties {pitchCurve {editorCurveData {keys [{time=0}]}}, volumeCurve {editorCurveData {keys [<at 0 0 RCIM_Cubic>,<at 1 1>]}}}
 }
 {id=Modulator_0, children [WavePlayer_0], properties {pitchMax=1.1, volumeMax=1, volumeMin=1}}
 {id=QualityLevel_0, children [Enveloper_0,Enveloper_0,null,null]}
 {id=WavePlayer_0, properties {bLooping=true, soundWaveAssetPtr=&/Energy_Field_Close_Loop}}
]
properties {
 attenuationSettings=&/Energy_Field_Loop_Close_Att
 bOverrideConcurrency=true
 maxDistance=1150
 priority=0
 soundClassObject=/Game/Sounds/FortSoundClassesAndModes/Fort_LTM_BR_InGame_Ambience
 volumeMultiplier=1
 concurrencyOverrides {maxCount=3, resolutionRule=StopQuietest, voiceStealReleaseTime=0.1}
}
