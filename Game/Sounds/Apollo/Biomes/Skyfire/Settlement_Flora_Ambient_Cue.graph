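# Sound cue for ambient flora in the Skyfire biome's settlement, routed to an outdoor ambient sound class.
# A mixer layers one of two randomly chosen looping beds with a slightly quieter one-shot branch.
# That branch loops endlessly, waiting a random few seconds before playing one of thirteen pitch-modulated, equally weighted one-shots.
# The cue overrides attenuation and concurrency, capping simultaneous instances at two.
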SoundCue /Game/Sounds/Apollo/Biomes/Skyfire/Settlement_Flora_Ambient_Cue
root=Mixer_0
nodes [
 {id=Delay_0, children [Modulator_0], properties {delayMax=6, delayMin=3}}
 {id=Looping_0, children [Delay_0]}
 {id=Mixer_0, children [Random_1,Looping_0], properties {inputVolume [1,0.8]}}
 {id=Modulator_0, children [Random_0], properties {pitchMax=1.3}}
 {
  id=Random_0
  children [
   WavePlayer_0
   WavePlayer_1
   WavePlayer_2
   WavePlayer_3
   WavePlayer_4
   WavePlayer_5
   WavePlayer_6
   WavePlayer_7
   WavePlayer_8
   WavePlayer_9
   WavePlayer_10
   WavePlayer_11
   WavePlayer_12
  ]
  properties {weights [1,1,1,1,1,1,1,1,1,1,1,1,1]}
 }
 {id=Random_1, children [WavePlayer_13, WavePlayer_14], properties {weights [1,1]}}
 {id=WavePlayer_0, properties {soundWaveAssetPtr=&/Settlement_Flora_OneShot_10}}
 {id=WavePlayer_1, properties {soundWaveAssetPtr=&/Settlement_Flora_OneShot_11}}
 {id=WavePlayer_10, properties {soundWaveAssetPtr=&/Settlement_Flora_OneShot_07}}
 {id=WavePlayer_11, properties {soundWaveAssetPtr=&/Settlement_Flora_OneShot_08}}
 {id=WavePlayer_12, properties {soundWaveAssetPtr=&/Settlement_Flora_OneShot_09}}
 {id=WavePlayer_13, properties {bLooping=true, soundWaveAssetPtr=&/Settlement_Flora_Loop_01}}
 {id=WavePlayer_14, properties {bLooping=true, soundWaveAssetPtr=&/Settlement_Flora_Loop_02}}
 {id=WavePlayer_2, properties {soundWaveAssetPtr=&/Settlement_Flora_OneShot_12}}
 {id=WavePlayer_3, properties {soundWaveAssetPtr=&/Settlement_Flora_OneShot_13}}
 {id=WavePlayer_4, properties {soundWaveAssetPtr=&/Settlement_Flora_OneShot_01}}
 {id=WavePlayer_5, properties {soundWaveAssetPtr=&/Settlement_Flora_OneShot_02}}
 {id=WavePlayer_6, properties {soundWaveAssetPtr=&/Settlement_Flora_OneShot_03}}
 {id=WavePlayer_7, properties {soundWaveAssetPtr=&/Settlement_Flora_OneShot_04}}
 {id=WavePlayer_8, properties {soundWaveAssetPtr=&/Settlement_Flora_OneShot_05}}
 {id=WavePlayer_9, properties {soundWaveAssetPtr=&/Settlement_Flora_OneShot_06}}
]
properties {
 bOverrideAttenuation=true
 bOverrideConcurrency=true
 maxDistance=7804.7788
 soundClassObject=&Fort_Ambient_Outdoors_Athena
 volumeMultiplier=0.601
 attenuationOverrides {falloffDistance=5837.7754, stereoSpread=9274.9814, attenuationShapeExtents {x=1967.0033}}
 concurrencyOverrides {maxCount=2, voiceStealReleaseTime=1}
}
